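# Physics asset for the face accessory of the medium female Pirate 01 player cosmetic, defining simulation bodies for its dangling feathers.
# Three feather chains (A and B with three bones, C with two) use one capsule per bone, with a small capsule on the head as their root.
# Each chain links bone to bone back to the head through identically limited constraints, allowing wider bend on one swing axis and little twist.
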PhysicsAsset /BRCosmetics/Characters/Player/Female/Medium/Bodies/F_MED_Pirate01/Meshes/Parts/F_MED_Pirate_01_FaceAcc_Physics
bodies {
 dyn_feather_A_01 [{shapeType=Capsule, length=5.3319, radius=2, center=<1 0 0>, rotation=<rot 90 0 0>}]
 dyn_feather_A_02 [{shapeType=Capsule, length=5.013, radius=2, center=<1 0 0>, rotation=<rot 90 0 0>}]
 dyn_feather_A_03 [{shapeType=Capsule, length=5.6743, radius=2, center=<1 0 0>, rotation=<rot 90 0 0>}]
 dyn_feather_B_01 [{shapeType=Capsule, length=6.805, radius=2, center=<1 0 0>, rotation=<rot 90 0 0>}]
 dyn_feather_B_02 [{shapeType=Capsule, length=3.2657, radius=2, center=<1 0 0>, rotation=<rot 90 0 0>}]
 dyn_feather_B_03 [{shapeType=Capsule, length=3.4881, radius=2, center=<1 0 0>, rotation=<rot 90 0 0>}]
 dyn_feather_C_01 [{shapeType=Capsule, length=2.9897, radius=2, center=<1 0 0>, rotation=<rot 90 0 0>}]
 dyn_feather_C_02 [{shapeType=Capsule, length=2.9703, radius=2, center=<1 0 0>, rotation=<rot 90 0 0>}]
 head [{shapeType=Capsule, length=1, radius=1}]
}
constraints [
 {bone1Name=dyn_feather_A_01, bone2Name=head, swing1LimitDegrees=10, swing2LimitDegrees=45, twistLimitDegrees=5}
 {bone1Name=dyn_feather_A_02, bone2Name=dyn_feather_A_01, swing1LimitDegrees=10, swing2LimitDegrees=45, twistLimitDegrees=5}
 {bone1Name=dyn_feather_A_03, bone2Name=dyn_feather_A_02, swing1LimitDegrees=10, swing2LimitDegrees=45, twistLimitDegrees=5}
 {bone1Name=dyn_feather_B_01, bone2Name=head, swing1LimitDegrees=10, swing2LimitDegrees=45, twistLimitDegrees=5}
 {bone1Name=dyn_feather_B_02, bone2Name=dyn_feather_B_01, swing1LimitDegrees=10, swing2LimitDegrees=45, twistLimitDegrees=5}
 {bone1Name=dyn_feather_B_03, bone2Name=dyn_feather_B_02, swing1LimitDegrees=10, swing2LimitDegrees=45, twistLimitDegrees=5}
 {bone1Name=dyn_feather_C_01, bone2Name=head, swing1LimitDegrees=10, swing2LimitDegrees=45, twistLimitDegrees=5}
 {bone1Name=dyn_feather_C_02, bone2Name=dyn_feather_C_01, swing1LimitDegrees=10, swing2LimitDegrees=45, twistLimitDegrees=5}
]
